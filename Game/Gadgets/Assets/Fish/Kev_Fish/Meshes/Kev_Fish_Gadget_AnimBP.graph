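# Animation Blueprint for the Kev Fish gadget on the shared Flopper gadget skeleton, adding procedural secondary motion to the spine.
# It feeds the local ref pose through two chained, partially blended Trail controllers, front spine then back spine, each with rotation limits, into the output pose.
# No animation assets or variables drive it; the event graph has no logic, and a front-spine Spring controller sits unconnected.
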AnimBlueprint /Game/Gadgets/Assets/Fish/Kev_Fish/Meshes/Kev_Fish_Gadget_AnimBP
target_skeleton=/Game/Gadgets/Assets/Flopper/Meshes/Flopper_Gadget_Skeleton
graphs [
 AnimGraph {nodes [
   n1 = Trail controller - Bone: dyn_spineFront_02!(ComponentPose=Local To Component(LocalPose=Local Space Ref Pose()), Alpha=0.2, RelaxationSpeedScale:float=4) -> (Pose) {
    maxDeltaTime=0.0333
    relaxationSpeedScale=4
    rotationLimits [{limitMax=<180 5 180>, limitMin=<-180 -5 -180>}, {limitMax=<180 5 180>, limitMin=<-180 -5 -180>}]
    rotationOffsets [<0 0 0>, <0 0 0>]
    trailBone {boneName=dyn_spineFront_02}
   }
   n2 = Trail controller - Bone: dyn_spineBack_02!(ComponentPose=@n1, Alpha=0.25, RelaxationSpeedScale:float=5) -> (Pose) {
    maxDeltaTime=0.0333
    relaxationSpeedScale=3
    rotationLimits [{limitMax=<5 15 65>, limitMin=<-5 -15 -65>}, {limitMax=<5 15 65>, limitMin=<-5 -15 -65>}]
    rotationOffsets [<0 0 0>, <0 0 0>]
    trailBone {boneName=dyn_spineBack_02}
   }
   Output Pose(Result=Component To Local(ComponentPose=@n2)) {layerGroup=None}
   Spring controller - Bone: dyn_spineFront_01!(ComponentPose:Component Space Pose Link Structure, Alpha:float=1) -> (Pose:Component Space Pose Link Structure) {
    is=title AnimGraphNode_SpringBone
    bLimitDisplacement=true
    maxDisplacement=1
    alphaBoolBlend {blendOption=Linear}
    springBone {boneName=dyn_spineFront_01}
   }
  ]}
 EventGraph {nodes [
   TryGetPawnOwner()
   Event BlueprintUpdateAnimation()
  ]}
]
variables []
